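# Niagara system for the GrillMoon mask pillar shockwave effect, an energy build-up that intensifies over roughly six to seven seconds.
# Flickering eye glow, energy thorns, an opening wave and repeating waves form the body, while GPU sparks get pulled back into the center and killed.
# Age-driven curves ramp brightness, pulse frequency and spawn rate; shared System Color and Curve values tint and scale layers.
# Rays and Wave02 emitters are disabled.
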NiagaraSystem /GrillMoon/Assets/FX/Shockwave/NS_GrillMoon_Mask_Pillar
dependencies {
 usedDataInterfaces [
  &NiagaraDataInterfaceSpriteRendererInfo
  &NiagaraDataInterfaceCamera
  &NiagaraDataInterfaceColorCurve
  &NiagaraDataInterfaceCurve
  &NiagaraDataInterfaceVector2DCurve
  &NiagaraDataInterfaceVectorCurve
  &NiagaraDataInterfacePlatformSet
 ]
 usedDynamicInputs [&ReturnCameraVector, &LerpLinearColors]
 usedModules [
  &SystemState
  &self:SystemSpawnScript.SystemScriptSource.SystemScriptGraph.NiagaraNodeAssignment_0.SetVariables_#1
  &self:SystemSpawnScript.SystemScriptSource.SystemScriptGraph.NiagaraNodeAssignment_2.SetVariables_#2
  &self:SystemSpawnScript.SystemScriptSource.SystemScriptGraph.NiagaraNodeAssignment_1.SetVariables_#3
  &self:SystemSpawnScript.SystemScriptSource.SystemScriptGraph.NiagaraNodeAssignment_3.SetVariables_#4
  &self:SystemSpawnScript.SystemScriptSource.SystemScriptGraph.NiagaraNodeAssignment_4.SetVariables_#5
 ]
}
emitters {
 EnergyThorn {
  emitterUpdateScript [
   &EmitterState(Inactive Response=::NewEnumerator1, Life Cycle Mode=::NewEnumerator1, Loop Behavior=::NewEnumerator1, Loop Duration=7, UseLoopDelay=false)
   &SpawnBurst_Instantaneous(Spawn Count=3, Use Spawn Probability=false, UseLoopCountLimit=true)
   &SpawnRate(SpawnRate=10, Use Spawn Probability=false)
  ]
  particleSpawnScript [
   &InitializeParticle(
    Lifetime Max=0.7
    Lifetime Min=0.4
    Lifetime Mode=::NewEnumerator1
    Mesh Scale Max=<2.4 2 0.7>
    Mesh Scale Min=<2 1.5 0.25>
    Mesh Scale Mode=::NewEnumerator2
    UsePositionOffset=true
    Write Material Random=true
   )
   &AddVelocity(
    Cone Angle=12
    Cone Axis=<0.5 0 1>
    Module.UseConeFalloff=false
    Module.UseVelDistribution=false
    Override Seed=false
    Velocity Mode=::NewEnumerator2
    Velocity Speed=1
   )
   &InitialMeshOrientation(
    Mesh Axis To Orient=<1 0 -1>
    Mesh Orientation Mode=::NewEnumerator2
    Rotation=&RandomRangeVector(Minimum=<0 0 0>, Maximum=<1 0 0>)
    Use Rotation Vector=true
   )
   &InitialMeshOrientation(
    Mesh Orientation Mode=::NewEnumerator2
    Rotation=&RandomRangeVector(Minimum=<0 0 0>, Maximum=<1 0 0>)
    Use Rotation Vector=true
   ) {enabled=false, module=InitialMeshOrientation001}
  ]
  particleUpdateScript [
   &ParticleState()
   &ScaleMeshSize(Scale Factor=&Multiply_Vector(
     A=&MakeVector(
      X=&FloatFromCurve(FloatCurve=DataInterface() {propertyValues {Curve {keys [
           <at 0 5 RCIM_Cubic>
           {arriveTangent=2.9874, interpMode=RCIM_Cubic, leaveTangent=2.9874, tangentMode=RCTM_User, time=0.05, value=0.1}
           {arriveTangent=2.9, interpMode=RCIM_Cubic, leaveTangent=2.9, tangentMode=RCTM_User, time=0.5, value=1.25}
           {arriveTangentWeight=0.5077, interpMode=RCIM_Cubic, tangentMode=RCTM_User, time=1, value=3}
          ]}}})
      Y=&FloatFromCurve(FloatCurve=DataInterface() {propertyValues {Curve {keys [{arriveTangent=7.6181, interpMode=RCIM_Cubic, leaveTangent=7.6181, tangentMode=RCTM_User, time=0, value=0}, <at 1 3 RCIM_Cubic>]}}})
      Z=&FloatFromCurve(FloatCurve=DataInterface() {propertyValues {Curve {keys [{arriveTangent=7.6181, interpMode=RCIM_Cubic, leaveTangent=7.6181, tangentMode=RCTM_User, time=0, value=0}, <at 1 3 RCIM_Cubic>]}}})
     )
     B=&VectorFromFloat(Value=$System.Curve)
    ))
   &ScaleColor(
    Scale Alpha=&FloatFromCurve(
     FloatCurve=DataInterface() {propertyValues {Curve {keys [
         {arriveTangent=2.8247, interpMode=RCIM_Cubic, leaveTangent=2.8247, tangentMode=RCTM_User, time=0, value=0.5}
         {arriveTangent=0.0384, interpMode=RCIM_Cubic, leaveTangent=0.0384, tangentMode=RCTM_User, time=0.4, value=1}
         {arriveTangent=0.0424, interpMode=RCIM_Cubic, leaveTangent=0.0424, tangentMode=RCTM_User, time=1, value=0}
        ]}}}
     Scale Curve=0.5
    )
    Scale RGB=&Multiply_VectorByFloat(
     Vector=&Multiply_VectorByFloat(
      Vector=<4 4 4>
      Float=&Waveform(
       "Waveform [1]"=::NewEnumerator0
       "[1] Frequency"=&Lerp_Float(
        A=1
        B=5
        Alpha=&FloatFromCurve(
         FloatCurve=DataInterface() {propertyValues {Curve {keys [
             {arriveTangent=0.3375, interpMode=RCIM_Cubic, leaveTangent=0.3375, tangentMode=RCTM_User, time=0, value=0}
             {arriveTangent=1.209, interpMode=RCIM_Cubic, leaveTangent=1.209, tangentMode=RCTM_User, time=0.95, value=0.5}
             {arriveTangent=8.2869, interpMode=RCIM_Cubic, leaveTangent=8.2869, tangentMode=RCTM_User, time=1, value=2}
            ]}}}
         CurveIndex=$System.NormalizedLoopAge
        )
       )
       "[1] Amplitude Min/Max"=<1 3>
      )
     )
     Float=&Lerp_Float(
      A=1
      B=5
      Alpha=&FloatFromCurve(
       FloatCurve=DataInterface() {propertyValues {Curve {keys [
           <at 0 2 RCIM_Cubic>
           {arriveTangent=0.3882, interpMode=RCIM_Cubic, leaveTangent=0.3882, tangentMode=RCTM_User, time=0.05, value=0}
           {arriveTangent=2.8914, interpMode=RCIM_Cubic, leaveTangent=2.8914, tangentMode=RCTM_User, time=0.8, value=0.625}
           {interpMode=RCIM_Cubic, tangentMode=RCTM_User, time=1, value=2}
          ]}}}
       CurveIndex=$System.NormalizedLoopAge
      )
     )
    )
    ScaleA=true
    ScaleRGB=true
   ) {enabled=false}
   &ScaleColor(
    Scale Alpha=&FloatFromCurve(
     FloatCurve=DataInterface() {propertyValues {Curve {keys [
         {arriveTangent=2.8247, interpMode=RCIM_Cubic, leaveTangent=2.8247, tangentMode=RCTM_User, time=0, value=0.5}
         {arriveTangent=0.0384, interpMode=RCIM_Cubic, leaveTangent=0.0384, tangentMode=RCTM_User, time=0.4, value=1}
         {arriveTangent=0.0424, interpMode=RCIM_Cubic, leaveTangent=0.0424, tangentMode=RCTM_User, time=1, value=0}
        ]}}}
     Scale Curve=0.5
    )
    Scale RGB=&Multiply_VectorByFloat(
     Vector=&Multiply_VectorByFloat(
      Vector=<4 4 4>
      Float=&Waveform(
       "Waveform [1]"=::NewEnumerator0
       "[1] Frequency"=&Lerp_Float(
        A=1
        B=5
        Alpha=&FloatFromCurve(
         FloatCurve=DataInterface() {propertyValues {Curve {keys [
             {arriveTangent=0.0115, interpMode=RCIM_Cubic, leaveTangent=0.0115, tangentMode=RCTM_User, time=0, value=0}
             {arriveTangent=0.2011, interpMode=RCIM_Cubic, leaveTangent=0.2011, tangentMode=RCTM_User, time=5.7, value=0.5}
             {arriveTangent=8.2869, interpMode=RCIM_Cubic, leaveTangent=8.2869, tangentMode=RCTM_User, time=6, value=2}
            ]}}}
         CurveIndex=$Engine.System.Age
        )
       )
       "[1] Amplitude Min/Max"=<1 3>
      )
     )
     Float=&Lerp_Float(
      A=1
      B=5
      Alpha=&FloatFromCurve(
       FloatCurve=DataInterface() {propertyValues {Curve {keys [
           <at 0 2 RCIM_Cubic>
           {
            arriveTangent=0.066
            arriveTangentWeight=0.104
            interpMode=RCIM_Cubic
            leaveTangent=0.066
            leaveTangentWeight=1.5603
            tangentMode=RCTM_User
            time=0.3
            value=0
           }
           {arriveTangent=0.5055, interpMode=RCIM_Cubic, leaveTangent=0.5055, tangentMode=RCTM_User, time=4.8, value=0.625}
           {interpMode=RCIM_Cubic, tangentMode=RCTM_User, time=6, value=2}
          ]}}}
       CurveIndex=$Engine.System.Age
      )
     )
    )
    ScaleA=true
    ScaleRGB=true
   ) {module=ScaleColor002}
   &Drag(Drag=6, Use Linear Drag=true, Use Rotational Drag=false)
   &SolveForcesAndVelocity(Clamp Velocity=false, Limit Acceleration=false)
   &ScaleColor(
    Color Value To Scale=$Particles.Color
    Scale Mode=::NewEnumerator1
    Scale RGBA=&MakeVector4_FromLinearColor(LinearColor=$System.Color)
    ScaleRGBA=true
   ) {module=ScaleColor001}
  ]
  rendererClasses [&NiagaraMeshRendererProperties]
  renderers [{rendererClass=&NiagaraMeshRendererProperties, rendererIndex=0}]
 }
 Eyeglow {
  emitterUpdateScript [
   &EmitterState()
   &SpawnBurst_Instantaneous(Use Spawn Probability=false, UseLoopCountLimit=true)
  ]
  particleSpawnScript [
   &InitializeParticle(
    Color=<rgba 0.15 0.0372 0 1>
    Lifetime=6
    Mesh Scale Mode=::NewEnumerator3
    Mesh Uniform Scale=0.4
    Position Offset=<6 0 12>
    Sprite Size Mode=::NewEnumerator3
    Uniform Sprite Size=1000
    UsePositionOffset=true
    Write Material Random=true
   )
   &InitializeParticle(
    Color=<rgba 0.15 0.0372 0 1>
    Lifetime=6
    Mesh Scale Mode=::NewEnumerator3
    Mesh Uniform Scale=0.15
    Position Offset=<6 0 12>
    Sprite Size Mode=::NewEnumerator3
    Uniform Sprite Size=500
    UsePositionOffset=true
    Write Material Random=true
   ) {enabled=false, module=InitializeParticle001}
   &InitialMeshOrientation(Mesh Orientation Mode=::NewEnumerator2, Orientation Vector=<0 0 1>, Rotation=<0 0.25 0>, Use Rotation Vector=true)
   &CameraOffset(Camera Offset Amount=10)
   &AddVelocity(Velocity=<0 0 0.001>)
  ]
  particleUpdateScript [
   &ParticleState()
   &ScaleMeshSize(Scale Factor=&VectorFromCurve(
     VectorCurve=DataInterface() {propertyValues {
       XCurve {keys [<at 0 0.25>,<at 0.1 1>,<at 0.9 1>,<at 1 0>]}
       YCurve {keys [<at 0 1.25>,<at 0.1 2.25>,<at 0.9 2.25>,<at 1 0>]}
       ZCurve {keys [<at 0 1.25>,<at 0.1 2.25>,<at 0.9 2.25>,<at 1 0>]}
      }}
     Scale Curve=&VectorFromFloat(Value=$System.Curve)
    ))
   &ScaleSpriteSize(
    Uniform Curve Scale=$System.Curve
    Uniform Curve Sprite Scale=DataInterface() {propertyValues {Curve {keys [
        <at 0 0>
        <at 0.1 1>
        <at 0.1512 0.7174>
        <at 0.2132 1>
        <at 0.2641 0.7539>
        <at 0.299 1>
        <at 0.3563 0.7037>
        <at 0.4151 1>
        <at 0.4564 0.8017>
        <at 0.5041 1>
        <at 0.5502 0.7561>
        <at 0.5836 1>
        <at 0.6551 0.7653>
        <at 0.6981 1>
        <at 0.7521 0.7721>
        <at 0.7871 1>
        <at 0.8475 0.7106>
        <at 0.9 1>
        {tangentMode=RCTM_User, time=1, value=0}
       ]}}}
   )
   /SetVariables(Particles.Size002:Vector2f=&Multiply_Vector2D_ByFloat(Float=$System.Curve, Vector2D=&Multiply_Vector2D(A=$Particles.SpriteSize, B=<0.6 0.9>))) {isSetParameters=true}
   &ScaleColor(
    Scale Alpha=&FloatFromCurve(FloatCurve=DataInterface() {propertyValues {Curve {keys [
         {arriveTangent=4.4475, time=0, value=0}
         {arriveTangent=5.3518, leaveTangent=-3.1774, time=0.1, value=1.2}
         {arriveTangent=-1.4972, leaveTangent=-1.4972, time=0.1523, value=0.7}
         {arriveTangent=-0.1667, leaveTangent=-0.1667, time=0.2052, value=1.1711}
         <at 0.258 0.7713>
         <at 0.2926 1.1775>
         {arriveTangent=-0.25, leaveTangent=-0.25, time=0.3608, value=0.689}
         <at 0.3865 1.193>
         <at 0.4263 0.8643>
         <at 0.4931 1.1191>
         {arriveTangent=-0.25, leaveTangent=-0.25, time=0.5428, value=0.7548}
         <at 0.5846 1.1063>
         <at 0.6571 0.8076>
         <at 0.6976 1.1685>
         {arriveTangent=0.1667, leaveTangent=0.1667, time=0.7542, value=0.7713}
         <at 0.7906 1.1138>
         <at 0.8535 0.8807>
         {arriveTangent=-3.3333, leaveTangent=-3.3333, time=0.9, value=1.2}
         <at 1 0>
        ]}}})
    Scale RGB=&MakeVectorFromLinearColorRGB(Color=&ColorFromCurve(ColorCurve=DataInterface() {propertyValues {
        AlphaCurve {keys []}
        BlueCurve {keys [<at 0.7887 0.43 RCIM_Cubic>,<at 0.9801 0.15 RCIM_Cubic>]}
        GreenCurve {keys [<at 0.7887 0.43 RCIM_Cubic>,<at 0.9801 0.15 RCIM_Cubic>]}
        RedCurve {keys [<at 0.7887 0.43 RCIM_Cubic>,<at 0.9801 0.15 RCIM_Cubic>]}
       }}))
    ScaleA=true
    ScaleRGB=true
   )
   /SetVariables(Particles.color002:LinearColor=&MakeLinearColorFromVectorAndFloat(
     "Vector (RGB)"=&MakeVectorFromLinearColorRGB(Color=$System.Color)
     "Float (Alpha)"=&FloatFromCurve(
      FloatCurve=DataInterface() {propertyValues {Curve {keys [<at 0 1>,<at 0.1 2>,<at 0.2005 1.7767>,<at 0.306 1.9927>,<at 0.3889 1.7548>,<at 0.5026 2.0076>,<at 0.5773 1.7477>,<at 0.6601 1.9791>,<at 0.7283 1.7134>,<at 0.8209 1.9848>,<at 1 0>]}}}
      Scale Curve=&FN_PickFloatDependOnPlatform(TargetPlatformActive=5, TargetPlatformNoActive=1)
     )
    )) {isSetParameters=true}
   /SetVariables(Particles.MeshColor:LinearColor=&MultiplyLinearColor(
     Linear Color=&MakeLinearColorFromVectorAndFloat(
      "Vector (RGB)"=&MakeVectorFromLinearColorRGB(Color=&ColorFromCurve(ColorCurve=DataInterface() {propertyValues {
          AlphaCurve {keys []}
          BlueCurve {keys [
            {arriveTangent=-1.8762, interpMode=RCIM_Cubic, leaveTangent=-1.8762, tangentMode=RCTM_User, time=0.1298, value=2}
            {interpMode=RCIM_Cubic, leaveTangent=0.2132, time=0.3441, value=1}
           ]}
          GreenCurve {keys [
            {arriveTangent=-0.4331, interpMode=RCIM_Cubic, leaveTangent=-0.4331, tangentMode=RCTM_User, time=0.1298, value=2}
            {interpMode=RCIM_Cubic, leaveTangent=2.0123, time=0.3441, value=1}
           ]}
          RedCurve {keys [
            {arriveTangent=-4.577, interpMode=RCIM_Cubic, leaveTangent=-4.577, tangentMode=RCTM_User, time=0.1298, value=2}
            {interpMode=RCIM_Cubic, leaveTangent=-2.7231, time=0.3441, value=1}
           ]}
         }}))
      "Float (Alpha)"=&FloatFromCurve(FloatCurve=DataInterface() {propertyValues {Curve {keys [
           {arriveTangent=-0.9, time=0, value=0}
           {arriveTangent=-0.0998, leaveTangent=-0.0998, time=0.05, value=1}
           {arriveTangent=-1.0872, leaveTangent=-1.0872, time=0.1804, value=0.7734}
           <at 0.2985 0.937>
           <at 0.4531 0.6834>
           <at 0.5813 0.9005>
           <at 0.7511 0.6392>
           <at 0.8564 0.8>
           {arriveTangent=-1.5912, leaveTangent=-1.5912, time=1, value=0}
          ]}}})
     )
     Scale Factor=&MakeVector4_FromLinearColor(LinearColor=$System.Color)
    )) {isSetParameters=true}
   &FN_Light_Attributes(
    Color=&ColorFromCurve(ColorCurve=DataInterface() {propertyValues {
       AlphaCurve {keys []}
       BlueCurve {keys [<at 0 0 RCIM_Cubic>,<at 0.0858 30>,<at 0.9 2>,<at 1 0 RCIM_Cubic>]}
       GreenCurve {keys [<at 0 0 RCIM_Cubic>,<at 0.0858 12.5>,<at 0.9 1.1328>,<at 1 0 RCIM_Cubic>]}
       RedCurve {keys [<at 0 0 RCIM_Cubic>,<at 0.0858 0>,<at 0.9 0.0729>,<at 1 0 RCIM_Cubic>]}
      }})
    Create Color Attribute=true
    Create Exponent Attribute=true
    Create Position Attribute=true
    Create Radius Attribute=true
    Exponent=3
    Radius=&FloatFromCurve(
     FloatCurve=DataInterface() {propertyValues {Curve {keys [<at 0 0.4>,<at 0.1 1.5>,<at 0.2 1>,<at 0.5 1>,<at 0.8 0.8>,<at 1 0.4>]}}}
     Scale Curve=&Multiply_Float(A=&MakeCustomFloatFromBool(Bool=$User.bIsFrontEndPreview, True Float=40, False Float=80), B=1)
    )
    Scale Color=&FloatFromCurve(
     FloatCurve=DataInterface() {propertyValues {Curve {keys [<at 0 0.5>,<at 0.1 1.1>,<at 0.1401 0.4933>,<at 0.2 1.3742>,<at 0.2354 0.5667>,<at 0.3 1.1>,<at 0.3404 0.4667>,<at 0.4098 1.3991>,<at 0.4342 0.48>,<at 0.5 1.1>,<at 0.5311 0.5133>,<at 0.5829 1.3306>,<at 0.6376 0.5133>,<at 0.7 1.1>,<at 0.741 0.5067>,<at 0.7989 1.4926>,<at 0.8459 0.3133>,<at 0.9 1.1>,<at 1 0>]}}}
     Scale Curve=5
    )
   )
   &SolveForcesAndVelocity(Clamp Velocity=false, Limit Acceleration=false)
   &ScaleColor(
    Scale Mode=::NewEnumerator1
    Scale RGBA=&MakeVector4_FromLinearColor(LinearColor=$System.Color)
    ScaleRGBA=true
   ) {module=ScaleColor001}
  ]
  rendererClasses [&NiagaraSpriteRendererProperties, &NiagaraMeshRendererProperties]
  renderers [
   {rendererClass=&NiagaraSpriteRendererProperties}
   {rendererClass=&NiagaraSpriteRendererProperties, rendererIndex=1}
   {rendererClass=&NiagaraMeshRendererProperties, rendererIndex=2}
  ]
 }
 Rays {
  bEnabled=false
  emitterUpdateScript [
   &EmitterState(Life Cycle Mode=::NewEnumerator1, Loop Delay=0.7, UseLoopDelay=true)
   &SpawnBurst_Instantaneous(Use Spawn Probability=false, UseLoopCountLimit=true)
   &SpawnBurst_Instantaneous(Spawn Time=0.15, Use Spawn Probability=false, UseLoopCountLimit=true) {module=SpawnBurst_Instantaneous001}
   &SpawnBurst_Instantaneous(Spawn Time=0.9, Use Spawn Probability=false, UseLoopCountLimit=true) {module=SpawnBurst_Instantaneous003}
   &SpawnBurst_Instantaneous(Spawn Time=1.8, Use Spawn Probability=false, UseLoopCountLimit=true) {module=SpawnBurst_Instantaneous005}
   &SpawnBurst_Instantaneous(Spawn Time=2.4, Use Spawn Probability=false, UseLoopCountLimit=true) {module=SpawnBurst_Instantaneous007}
  ]
  particleSpawnScript [
   &InitializeParticle(
    Color=<rgba 0 8.77 20 0.25>
    Lifetime=&RandomRangeFloat(Minimum=0.2, Maximum=0.3)
    Sprite Size Max=<350 600>
    Sprite Size Min=<200 400>
    Sprite Size Mode=::NewEnumerator2
    UsePositionOffset=false
    Write Material Random=true
   )
   &SubUVAnimation(
    Evaluation Type=::NewEnumerator0
    "Sprite Renderer "=DataInterface() {propertyValues {SpriteRenderer=&self:Rays_2.NiagaraSpriteRendererProperties_2}}
    SubUV Animation Mode:ENiagara_SubUVLookupModeV2=::NewEnumerator2
    UseEndFrame=false
    UseStartFrame=false
   )
   &AddVelocity(
    Cone Angle=30
    Cone Axis=<0.5 0 1>
    Module.UseConeFalloff=false
    Module.UseVelDistribution=false
    Override Seed=false
    Velocity Mode=::NewEnumerator2
    Velocity Speed=0.1
   )
   &CameraOffset(Camera Offset Amount=&RandomRangeFloat(Minimum=-25, Maximum=15))
  ]
  particleUpdateScript [
   &ParticleState(Kill Particles When Lifetime Has Elapsed=false, Let Infinitely Lived Particles Die When Emitter Deactivates=true)
   &SolveForcesAndVelocity(Clamp Velocity=false, Limit Acceleration=false)
   &ScaleColor(
    Scale Alpha=&FloatFromCurve(FloatCurve=DataInterface() {propertyValues {Curve {keys [{interpMode=RCIM_Cubic, tangentMode=RCTM_User, time=0, value=0}, {interpMode=RCIM_Cubic, tangentMode=RCTM_User, time=1, value=1}]}}})
    Scale RGB=&MakeVectorFromLinearColorRGB(Color=&MultiplyLinearColor(
      Linear Color=&Multiply_LinearColorByFloat(
       Linear Color=&ColorFromCurve(ColorCurve=DataInterface() {propertyValues {
          AlphaCurve {keys [<at 0 1 RCIM_Cubic>]}
          BlueCurve {keys [<at 0.0426 10 RCIM_Cubic>, {arriveTangent=-12.0954, interpMode=RCIM_Cubic, leaveTangent=-12.0954, time=0.2988, value=5}, <at 0.8197 0.6>]}
          GreenCurve {keys [<at 0.0426 10 RCIM_Cubic>, {arriveTangent=-12.0954, interpMode=RCIM_Cubic, leaveTangent=-12.0954, time=0.2988, value=5}, <at 0.8197 0.6>]}
          RedCurve {keys [<at 0.0426 10 RCIM_Cubic>, {arriveTangent=-12.0954, interpMode=RCIM_Cubic, leaveTangent=-12.0954, time=0.2988, value=5}, <at 0.8197 0.6>]}
         }})
       Float=1
      )
      Scale Factor=<1 1 1 1>
     ))
    ScaleA=true
    ScaleRGB=true
   )
   &ScaleSpriteSize(
    Scale Sprite Size Mode=::NewEnumerator0
    Uniform Scale Factor=&FloatFromCurve(FloatCurve=DataInterface() {propertyValues {Curve {keys [
         {arriveTangent=2.0201, interpMode=RCIM_Cubic, leaveTangent=2.0201, tangentMode=RCTM_User, time=0, value=0}
         {arriveTangent=-0.026, interpMode=RCIM_Cubic, leaveTangent=-0.026, tangentMode=RCTM_User, time=0.6, value=1}
        ]}}})
   )
   &ScaleSpriteSize(
    Non-Uniform Curve Sprite Scale=DataInterface() {propertyValues {
      XCurve {keys [
        {arriveTangent=-1.6611, interpMode=RCIM_Cubic, leaveTangent=-1.6611, tangentMode=RCTM_User, time=0.2, value=1.2}
        {arriveTangent=-0.1324, interpMode=RCIM_Cubic, leaveTangent=-0.1324, tangentMode=RCTM_User, time=1, value=0.7}
       ]}
      YCurve {keys [{interpMode=RCIM_Cubic, tangentMode=RCTM_User, time=0.4, value=1}]}
     }}
    Scale Sprite Size Mode=::NewEnumerator3
   ) {module=ScaleSpriteSize001}
   &DynamicMaterialParameters(
    Index 0 Param 2=1
    Index 0 Param 3=1
    Index 0 Param 4=1
    Module.Index0Param0Write=true
    Module.Index0Param1Write=true
    Module.Index0Param2Write=true
    Module.Index0Param3Write=true
   )
  ]
  rendererClasses [&NiagaraSpriteRendererProperties]
  renderers [{rendererClass=&NiagaraSpriteRendererProperties}]
 }
 Wave02 {
  bEnabled=false
  emitterUpdateScript [
   &EmitterState(Life Cycle Mode=::NewEnumerator1, Loop Behavior=::NewEnumerator1, Loop Delay=$System.Delay, Loop Duration=1, UseLoopDelay=true)
   &SpawnBurst_Instantaneous(Spawn Count=1, Use Spawn Probability=false)
  ]
  particleSpawnScript [
   &InitializeParticle(Color=$System.Color, Lifetime=6, Mesh Scale Mode=::NewEnumerator3, Mesh Uniform Scale=2, UsePositionOffset=true, Write Material Random=true)
   &InitialMeshOrientation(Orientation Coordinate Space=::World, Orientation Vector=<0 0 1>, Use Orientation Vector=true, Use Rotation Vector=false)
   &CameraOffset(Camera Offset Amount=-50)
  ]
  particleUpdateScript [
   &ParticleState()
   &ScaleColor(
    Linear Color Curve=DataInterface() {propertyValues {
      AlphaCurve {keys [<at 0 0>,<at 0.1 0.25>]}
      BlueCurve {keys [<at 0.1161 1>]}
      GreenCurve {keys [<at 0.1161 1>]}
      RedCurve {keys [<at 0.1161 1>]}
     }}
    Scale Mode=::NewEnumerator2
   )
   &DynamicMaterialParameters(
    Index 0 Param 1=&FloatFromCurve(FloatCurve=DataInterface() {propertyValues {Curve {keys [<at 0 1>,<at 0.2 0>]}}}, CurveIndex=$Particles.Age)
    Index 0 Param 2=&FloatFromCurve(FloatCurve=DataInterface() {propertyValues {Curve {keys [{arriveTangent=-0.5818, interpMode=RCIM_Cubic, leaveTangent=-0.5818, tangentMode=RCTM_User, time=0, value=1}, <at 1 0.9>]}}})
    Index 0 Param 3=&FloatFromCurve(FloatCurve=DataInterface() {propertyValues {Curve {keys [<at 0.15 1>,<at 1 1.2>]}}})
    Module.Param0WriteEnabled=true
    Module.Param1WriteEnabled=true
    Module.Param2WriteEnabled=true
    Module.Param3WriteEnabled=false
   )
   &ScaleColor(
    Scale Alpha=&FNTimeOfDayScalar(Night Scale=0.7, Morning Scale=0.5, Evening Scale=0.5)
    ScaleA=true
    ScaleRGB=true
   ) {module=ScaleColor001}
   &ScaleMeshSize(Scale Factor=&VectorFromFloat(Value=&FloatFromCurve(
      FloatCurve=DataInterface() {propertyValues {Curve {keys [
          {interpMode=RCIM_Cubic, tangentMode=RCTM_User, time=0, value=0}
          {interpMode=RCIM_Cubic, tangentMode=RCTM_User, time=0.1, value=1}
          {interpMode=RCIM_Cubic, tangentMode=RCTM_User, time=0.8, value=1}
          {interpMode=RCIM_Cubic, tangentMode=RCTM_User, time=1, value=0}
         ]}}}
      CurveIndex=$System.Curve
      Scale Curve=&Lerp_Float(A=0.1, B=1, Alpha=$System.Curve)
     )))
  ]
  rendererClasses [&NiagaraMeshRendererProperties]
  renderers [{rendererClass=&NiagaraMeshRendererProperties, rendererIndex=0}]
 }
 Wave02001 {
  emitterUpdateScript [
   &EmitterState(Inactive Response=::NewEnumerator1, Life Cycle Mode=::NewEnumerator1, Loop Behavior=::NewEnumerator1, Loop Duration=7, UseLoopDelay=false)
   &SpawnRate(SpawnRate=2, Use Spawn Probability=false)
   &SpawnBurst_Instantaneous(Spawn Count=1, Use Spawn Probability=false) {enabled=false}
  ]
  particleSpawnScript [
   &InitializeParticle(
    Lifetime=&Lerp_Float(
     A=2
     B=1.25
     Alpha=&FloatFromCurve(
      FloatCurve=DataInterface() {propertyValues {Curve {keys [
          {arriveTangent=0.3882, interpMode=RCIM_Cubic, leaveTangent=0.3882, tangentMode=RCTM_User, time=0, value=0}
          {arriveTangent=2.8914, interpMode=RCIM_Cubic, leaveTangent=2.8914, tangentMode=RCTM_User, time=0.7, value=0.875}
          {interpMode=RCIM_Cubic, tangentMode=RCTM_User, time=1, value=2}
         ]}}}
      CurveIndex=$System.NormalizedLoopAge
     )
    )
    Mesh Scale Mode=::NewEnumerator3
    Mesh Uniform Scale=2
    UsePositionOffset=false
    Write Material Random=true
   )
   &InitialMeshOrientation(
    Orientation Coordinate Space=::Local
    Orientation Vector=<0 0 1>
    Rotation=&RandomRangeVector(Minimum=<0 0 0>, Maximum=<0 0 1>)
    Use Orientation Vector=true
    Use Rotation Vector=true
   )
   &CameraOffset()
  ]
  particleUpdateScript [
   &ParticleState()
   &ScaleColor(
    Linear Color Curve=DataInterface() {propertyValues {
      AlphaCurve {keys [<at 0 0>,<at 0.3576 0.5>,<at 1 0>]}
      BlueCurve {keys [<at 0.1161 1>]}
      GreenCurve {keys [<at 0.1161 1>]}
      RedCurve {keys [<at 0.1161 1>]}
     }}
    Scale Mode=::NewEnumerator2
   )
   &DynamicMaterialParameters(
    Index 0 Param 1=&FloatFromCurve(FloatCurve=DataInterface() {propertyValues {Curve {keys [
         {arriveTangent=-0.325, interpMode=RCIM_Cubic, leaveTangent=-0.325, tangentMode=RCTM_User, time=0, value=-0.7}
         {arriveTangent=3.0524, tangentMode=RCTM_User, time=1, value=0.6}
        ]}}})
    Index 0 Param 2=&FloatFromCurve(FloatCurve=DataInterface() {propertyValues {Curve {keys [{arriveTangent=0.0262, interpMode=RCIM_Cubic, leaveTangent=0.0262, tangentMode=RCTM_User, time=0, value=0.9}, <at 1 1>]}}})
    Index 0 Param 3=&FloatFromCurve(FloatCurve=DataInterface() {propertyValues {Curve {keys [<at 0 0>,<at 0.15 1>,<at 1 1.2>]}}})
    Module.Param0WriteEnabled=true
    Module.Param1WriteEnabled=true
    Module.Param2WriteEnabled=true
    Module.Param3WriteEnabled=false
   )
   &ScaleColor(
    Scale Alpha=&FNTimeOfDayScalar(Night Scale=0.7, Morning Scale=0.5, Evening Scale=0.5)
    Scale RGB=&VectorFromFloat(Value=&Lerp_Float(
      A=1
      B=2
      Alpha=&FloatFromCurve(
       FloatCurve=DataInterface() {propertyValues {Curve {keys [
           {arriveTangent=0.3882, interpMode=RCIM_Cubic, leaveTangent=0.3882, tangentMode=RCTM_User, time=0, value=0}
           {arriveTangent=2.8914, interpMode=RCIM_Cubic, leaveTangent=2.8914, tangentMode=RCTM_User, time=0.8, value=0.625}
           {interpMode=RCIM_Cubic, tangentMode=RCTM_User, time=1, value=2}
          ]}}}
       CurveIndex=$System.NormalizedLoopAge
      )
     ))
    ScaleA=true
    ScaleRGB=true
   ) {module=ScaleColor001}
   &ScaleMeshSize(Scale Factor=&VectorFromFloat(Value=&FloatFromCurve(FloatCurve=DataInterface() {propertyValues {Curve {keys [{interpMode=RCIM_Cubic, tangentMode=RCTM_User, time=0, value=1}, {interpMode=RCIM_Cubic, tangentMode=RCTM_User, time=1, value=0}]}}}))) {enabled=false}
   &ScaleMeshSize(Scale Factor=&VectorFromFloat(Value=&Lerp_Float(
      A=0.5
      B=1
      Alpha=&FloatFromCurve(
       FloatCurve=DataInterface() {propertyValues {Curve {keys [
           {arriveTangent=0.3882, interpMode=RCIM_Cubic, leaveTangent=0.3882, tangentMode=RCTM_User, time=0, value=0}
           {arriveTangent=2.8914, interpMode=RCIM_Cubic, leaveTangent=2.8914, tangentMode=RCTM_User, time=0.7, value=0.75}
           {interpMode=RCIM_Cubic, tangentMode=RCTM_User, time=1, value=2}
          ]}}}
       CurveIndex=$System.NormalizedLoopAge
      )
     ))) {module=ScaleMeshSize001}
   &ScaleMeshSize(Scale Factor=&VectorFromFloat(Value=&Lerp_Float(A=1, B=0.2, Alpha=$System.Curve))) {enabled=false, module=ScaleMeshSize002}
   &ScaleColor(Scale RGB=&MakeVectorFromLinearColorRGB(Color=$System.Color), ScaleA=true, ScaleRGB=true) {module=ScaleColor002}
  ]
  rendererClasses [&NiagaraMeshRendererProperties]
  renderers [{rendererClass=&NiagaraMeshRendererProperties, rendererIndex=0}]
 }
 WaveStart {
  emitterUpdateScript [
   &EmitterState(Inactive Response=::NewEnumerator1, Life Cycle Mode=::NewEnumerator1, Loop Behavior=::NewEnumerator1, Loop Duration=1, UseLoopDelay=false)
   &SpawnBurst_Instantaneous(Spawn Count=1, Use Spawn Probability=false)
  ]
  particleSpawnScript [
   &InitializeParticle(Lifetime=0.7, Mesh Scale Mode=::NewEnumerator3, Mesh Uniform Scale=1.6, UsePositionOffset=false, Write Material Random=true)
   &InitialMeshOrientation(Orientation Coordinate Space=::Local, Orientation Vector=<0 0 1>, Use Orientation Vector=true, Use Rotation Vector=false)
   &CameraOffset(Camera Offset Amount=25)
  ]
  particleUpdateScript [
   &ParticleState()
   &ScaleColor(
    Linear Color Curve=DataInterface() {propertyValues {
      AlphaCurve {keys [<at 0 0>,<at 0.3576 0.5>,<at 1 0>]}
      BlueCurve {keys [<at 0.1161 1>]}
      GreenCurve {keys [<at 0.1161 1>]}
      RedCurve {keys [<at 0.1161 1>]}
     }}
    Scale Mode=::NewEnumerator2
   )
   &DynamicMaterialParameters(
    Index 0 Param 1=&FloatFromCurve(FloatCurve=DataInterface() {propertyValues {Curve {keys [
         {arriveTangent=-3.4443, interpMode=RCIM_Cubic, leaveTangent=-3.4443, tangentMode=RCTM_User, time=0, value=0.7}
         {arriveTangent=0.2369, tangentMode=RCTM_User, time=1, value=-0.7}
        ]}}})
    Index 0 Param 2=&FloatFromCurve(FloatCurve=DataInterface() {propertyValues {Curve {keys [{arriveTangent=0.0262, interpMode=RCIM_Cubic, leaveTangent=0.0262, tangentMode=RCTM_User, time=0, value=0.9}, <at 1 1>]}}})
    Index 0 Param 3=&FloatFromCurve(FloatCurve=DataInterface() {propertyValues {Curve {keys [<at 0 0>,<at 0.15 1>,<at 1 1.2>]}}})
    Module.Param0WriteEnabled=true
    Module.Param1WriteEnabled=true
    Module.Param2WriteEnabled=true
    Module.Param3WriteEnabled=false
   )
   &ScaleColor(
    Scale Alpha=&FNTimeOfDayScalar(Night Scale=0.7, Morning Scale=0.5, Evening Scale=0.5)
    Scale RGB=&VectorFromFloat(Value=1)
    ScaleA=true
    ScaleRGB=true
   ) {module=ScaleColor001}
   &ScaleMeshSize(Scale Factor=&VectorFromFloat(Value=1)) {module=ScaleMeshSize001}
   &ScaleColor(Scale RGB=&MakeVectorFromLinearColorRGB(Color=$System.Color), ScaleA=true, ScaleRGB=true) {module=ScaleColor002}
  ]
  rendererClasses [&NiagaraMeshRendererProperties]
  renderers [{rendererClass=&NiagaraMeshRendererProperties, rendererIndex=0}]
 }
 par {
  simTarget=GPUComputeSim
  emitterUpdateScript [
   &EmitterState()
   &SpawnRate(
    SpawnRate=&Lerp_Float(
     A=5
     B=100
     Alpha=&FloatFromCurve(
      FloatCurve=DataInterface() {propertyValues {Curve {keys [
          {arriveTangent=0.1484, interpMode=RCIM_Cubic, leaveTangent=0.1484, tangentMode=RCTM_User, time=0, value=0}
          {arriveTangent=0.2259, interpMode=RCIM_Cubic, leaveTangent=0.2259, tangentMode=RCTM_User, time=4.8, value=0.5}
          {interpMode=RCIM_Cubic, tangentMode=RCTM_User, time=6, value=2}
         ]}}}
      CurveIndex=$Engine.System.Age
     )
    )
    Use Spawn Probability=false
   )
   &SpawnRate(
    SpawnRate=&Lerp_Float(
     A=5
     B=100
     Alpha=&FloatFromCurve(
      FloatCurve=DataInterface() {propertyValues {Curve {keys [
          {arriveTangent=0.3882, interpMode=RCIM_Cubic, leaveTangent=0.3882, tangentMode=RCTM_User, time=0, value=0}
          {arriveTangent=1.945, interpMode=RCIM_Cubic, leaveTangent=1.945, tangentMode=RCTM_User, time=0.7, value=0.75}
          {interpMode=RCIM_Cubic, tangentMode=RCTM_User, time=1, value=2}
         ]}}}
      CurveIndex=$Emitter.NormalizedLoopAge
     )
    )
    Use Spawn Probability=false
   ) {enabled=false, module=SpawnRate001}
   &SpawnBurst_Instantaneous(Spawn Count=10, Use Spawn Probability=false, UseLoopCountLimit=false)
  ]
  particleSpawnScript [
   &InitializeParticle(
    Color=<rgba 5 5 5 1>
    Lifetime=&Multiply_Float(
     A=&RandomRangeFloat(Minimum=0.5, Maximum=0.8)
     B=&Lerp_Float(
      A=1
      B=0.6
      Alpha=&FloatFromCurve(
       FloatCurve=DataInterface() {propertyValues {Curve {keys [
           {arriveTangent=0.3882, interpMode=RCIM_Cubic, leaveTangent=0.3882, tangentMode=RCTM_User, time=0, value=0}
           {arriveTangent=2.8914, interpMode=RCIM_Cubic, leaveTangent=2.8914, tangentMode=RCTM_User, time=0.7, value=0.875}
           {interpMode=RCIM_Cubic, tangentMode=RCTM_User, time=1, value=2}
          ]}}}
       CurveIndex=$System.NormalizedLoopAge
      )
     )
    )
    Sprite Size Max=<4 8>
    Sprite Size Min=<1.5 2>
    Sprite Size Mode=::NewEnumerator2
    UsePositionOffset=false
    Write Material Random=true
   )
   &InitializeParticle(
    Color=<rgba 5 5 5 1>
    Lifetime=&Multiply_Float(
     A=&RandomRangeFloat(Minimum=0.5, Maximum=0.8)
     B=&Lerp_Float(
      A=1
      B=0.3
      Alpha=&FloatFromCurve(
       FloatCurve=DataInterface() {propertyValues {Curve {keys [
           {arriveTangent=0.3882, interpMode=RCIM_Cubic, leaveTangent=0.3882, tangentMode=RCTM_User, time=0, value=0}
           {arriveTangent=2.8914, interpMode=RCIM_Cubic, leaveTangent=2.8914, tangentMode=RCTM_User, time=0.7, value=0.875}
           {interpMode=RCIM_Cubic, tangentMode=RCTM_User, time=1, value=2}
          ]}}}
       CurveIndex=$System.NormalizedLoopAge
      )
     )
    )
    Sprite Size Max=<4 8>
    Sprite Size Min=<1.5 2>
    Sprite Size Mode=::NewEnumerator2
    UsePositionOffset=false
    Write Material Random=true
   ) {enabled=false, module=InitializeParticle001}
   &SphereLocation(
    MaskSpawnGroup=false
    Override Seed=false
    Sphere Radius=&Lerp_Float(A=20, B=80, Alpha=$System.Curve)
    Surface Expansion Mode=::NewEnumerator2
    Surface Only=true
    Surface Only Band Thickness=1
    UseNonUniformScale=false
    UseOffset=false
    UseSphereOrientation=false
   )
   &AddVelocityFromPoint(Module.Use Falloff Distance=false, Velocity Strength=&RandomRangeFloat(Minimum=150, Maximum=200))
   &CameraOffset(Camera Offset Amount=10)
  ]
  particleUpdateScript [
   &ParticleState()
   &CurlNoiseForce(
    Noise Frequency=200
    Noise Strength=&Multiply_Float(
     A=1500
     B=&FloatFromCurve(FloatCurve=DataInterface() {propertyValues {Curve {keys [<at 0 0>,<at 0.2 1>,<at 0.3 0>]}}})
    )
    Pan Noise=false
   )
   &Drag(Drag=&RandomRangeFloat(Minimum=1, Maximum=3), Use Linear Drag=true, Use Rotational Drag=false)
   &PointForce(
    Force Falloff Distance:NiagaraFloat=&RandomRangeFloat(Minimum=500, Maximum=400)
    Force Strength=&Multiply_Float(
     A=&RandomRangeFloat(Minimum=-2500, Maximum=-3500)
     B=&Lerp_Float(
      A=1
      B=2
      Alpha=&FloatFromCurve(
       FloatCurve=DataInterface() {propertyValues {Curve {keys [
           {arriveTangent=0.3882, interpMode=RCIM_Cubic, leaveTangent=0.3882, tangentMode=RCTM_User, time=0, value=0}
           {arriveTangent=2.8914, interpMode=RCIM_Cubic, leaveTangent=2.8914, tangentMode=RCTM_User, time=0.8, value=0.625}
           {interpMode=RCIM_Cubic, tangentMode=RCTM_User, time=1, value=2}
          ]}}}
       CurveIndex=$System.NormalizedLoopAge
      )
     )
    )
    Module.Use Falloff=true
    Module.UseLinearForce=false
   )
   &SolveForcesAndVelocity(Clamp Velocity=false, Limit Acceleration=false)
   &KillParticlesInVolume(Sphere Radius=15)
   &ScaleSpriteSizeBySpeed(Max Scale Factor=<0.2 10>, Min Scale Factor=<1 0.5>)
   &ScaleSpriteSize(Scale Factor=&MakeVector2D(
     X=&FloatFromCurve(FloatCurve=DataInterface() {propertyValues {Curve {keys [<at 0 0>,<at 0.1 1>,<at 1 0>]}}}, Scale Curve=1.75)
     Y=&FloatFromCurve(FloatCurve=DataInterface() {propertyValues {Curve {keys [<at 0 1>,<at 1 0>]}}}, Scale Curve=1.75)
    ))
   &ScaleColor(
    Linear Color Curve=DataInterface() {propertyValues {
      AlphaCurve {keys [<at 0 1>]}
      BlueCurve {keys [<at 0 1>,<at 0.5211 1>,<at 0.8598 1>]}
      GreenCurve {keys [<at 0 1>,<at 0.5211 1>,<at 0.8598 1>]}
      RedCurve {keys [<at 0 1>,<at 0.5211 1>,<at 0.8598 1>]}
     }}
    Scale Mode=::NewEnumerator2
   )
   &ScaleColor(
    Scale Alpha=&FNTimeOfDayScalar(Night Scale=0.5, Morning Scale=0.5, Evening Scale=0.5)
    Scale RGB=&VectorFromFloat(Value=&FloatFromCurve(FloatCurve=DataInterface() {propertyValues {Curve {keys [<at 0 0.5>,<at 0.4 1>]}}}, Scale Curve=20))
    ScaleA=true
    ScaleRGB=true
   ) {module=ScaleColor001}
   &ScaleColor(
    Scale Mode=::NewEnumerator1
    Scale RGBA=&MakeVector4_FromLinearColor(LinearColor=$System.Color)
    ScaleRGBA=true
   ) {module=ScaleColor002}
  ]
  rendererClasses [&NiagaraSpriteRendererProperties]
  renderers [{rendererClass=&NiagaraSpriteRendererProperties}]
 }
}
summary {
 systemName=NS_GrillMoon_Mask_Pillar
 emitters [
  {bEnabled=false, emitterName=Rays, rendererClasses [&NiagaraSpriteRendererProperties]}
  {bEnabled=false, emitterName=Wave02, rendererClasses [&NiagaraMeshRendererProperties]}
  {emitterName=Eyeglow, rendererClasses [&NiagaraSpriteRendererProperties, &NiagaraMeshRendererProperties]}
  {emitterName=EnergyThorn, rendererClasses [&NiagaraMeshRendererProperties]}
  {emitterName=par, simTarget=GPUComputeSim, rendererClasses [&NiagaraSpriteRendererProperties]}
  {emitterName=Wave02001, rendererClasses [&NiagaraMeshRendererProperties]}
  {emitterName=WaveStart, rendererClasses [&NiagaraMeshRendererProperties]}
 ]
 userVariables [
  User.bIsFrontEnd {type=NiagaraBool}
  User.bIsFrontEndPreview {type=NiagaraBool}
  User.Bounds {type=Vector3f, description="", defaultValue {struct=Vector3f, value=<400 400 900>}}
  User.Charge {type=NiagaraFloat, defaultValue=0, description=""}
 ]
}
